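# Niagara skydiving contrail cosmetic showing two burgers on skeletal mesh bones or sockets, trailing food bits, ketchup, mustard and grease.
# Mesh_SkeletalMesh layers buns, patty, cheese and toppings with sine bobbing and spin, spreading them as User.ParaGlideLeanAlpha rises.
# That alpha also drives debris spawn rates and speeds, ketchup streak shape and grease opacity.
# Partitioned particles select each attachment, debris launches against the contrail direction, and time-of-day color tints several emitters.
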NiagaraSystem /BRCosmetics/Athena/Effects/Particles/Skydiving/NS_Contrail_BoarGore/NS_Contrail_BoarGore
dependencies {
 usedDataInterfaces [
  &NiagaraDataInterfaceArrayInt32
  &NiagaraDataInterfaceCurve
  &NiagaraDataInterfaceVector2DCurve
  &NiagaraDataInterfaceMeshRendererInfo
  &NiagaraDataInterfaceCamera
  &NiagaraDataInterfaceVectorCurve
  &NiagaraDataInterfaceSkeletalMesh
 ]
 usedDynamicInputs []
 usedModules [
  &self:SystemSpawnScript.SystemScriptSource.SystemScriptGraph.NiagaraNodeAssignment_0.SetVariables_#1
  &SystemState
  /Game/Effects/Niagara/Modules/Contrails/FN_Contrail_Setup
  &FN_Integrator
 ]
}
emitters {
 FN_Debug {
  bEnabled=false
  emitterUpdateScript [
   &EmitterState()
   &SpawnBurst_Instantaneous(Spawn Count=1, Use Spawn Probability=false)
  ]
  particleSpawnScript [/Niagara/Modules/Spawn/Initialization/InitializeParticle(
    Lifetime=5
    Module.Write Color=true
    Module.Write Lifetime=true
    Module.Write Mass=true
    Module.Write Material Random=true
    Module.Write Position=false
    Module.Write Scale=false
    Module.Write SpriteRotation=false
    Module.Write SpriteSize=true
    Position=&SimulationPosition()
    Sprite Size=<100 100>
   )]
  particleUpdateScript [
   &ParticleState()
   /SetVariables(Particles.Debug:NiagaraFloat=&RemapRange(Input Value To Remap=$User.ParaGlideLeanAlpha, Input Min=0.16, Input Max=1, Output Min=0, Output Max=1)) {isSetParameters=true}
   &DynamicMaterialParameters(
    Index 0 Param 1=$Particles.Debug
    Module.Param0WriteEnabled=true
    Module.Param1WriteEnabled=true
    Module.Param2WriteEnabled=true
    Module.Param3WriteEnabled=true
   )
  ]
  rendererClasses [&NiagaraSpriteRendererProperties]
  renderers [{rendererClass=&NiagaraSpriteRendererProperties}]
 }
 FoodBits {
  emitterUpdateScript [
   &EmitterState(Life Cycle Mode=::NewEnumerator1, Loop Behavior=::NewEnumerator1, Loop Duration Mode=::NewEnumerator1, UseLoopDelay=false)
   &SpawnRate(SpawnRate=&Lerp_Float(A=16, B=32, Alpha=$User.ParaGlideLeanAlpha), Use Spawn Probability=false)
   &SpawnRate(SpawnGroup=2, SpawnRate=&Lerp_Float(A=0, B=32, Alpha=$User.ParaGlideLeanAlpha), Use Spawn Probability=false) {module=SpawnRate001}
   &SpawnRate(SpawnGroup=1, SpawnRate=&Lerp_Float(A=0, B=32, Alpha=$User.ParaGlideLeanAlpha), Use Spawn Probability=false) {module=SpawnRate002}
  ]
  particleSpawnScript [
   /SetVariables(Particles.VisibilityTag=&SpawnGroup()) {isSetParameters=true}
   &PartitionParticles(Partition Mode:ENiagara_PartitionMode=::NewEnumerator1)
   /Niagara/Modules/Spawn/Initialization/V2/InitializeParticle(
    Color Mode=::NewEnumerator0
    Lifetime Max=1.4
    Lifetime Min=0.6
    Lifetime Mode=::NewEnumerator1
    Mesh Scale Mode=::NewEnumerator3
    Mesh Uniform Scale=&MaskFloatBySpawnGroup(
     Masked Float=&RandomRangeFloat(Minimum=0.2, Maximum=0.45)
     Passthrough Float=&RandomRangeFloat(Minimum=0.24, Maximum=0.585)
    )
    UsePositionOffset=false
    Write Material Random=true
   )
   &ShapeLocation(
    Cylinder Height=20.0817
    Cylinder Radius=23.0564
    MaskSpawnGroup=false
    Offset=<0 0 23.9639>
    Offset Mode=::NewEnumerator0
    Override Seed=false
    Shape Primitive=::NewEnumerator1
    Surface Only=false
    UseOwnerScale=true
   )
   &AddVelocity(
    Cone Angle=15
    Cone Axis=&Multiply_VectorByFloat(Vector=$System.FN_Contrail_Setup.Direction Vector, Float=-1)
    Module.UseConeFalloff=false
    Module.UseVelDistribution=false
    Override Seed=false
    Velocity Mode=::NewEnumerator2
    Velocity Speed=&Lerp_Float(
     A=&RandomRangeFloat(Minimum=225, Maximum=425)
     B=&RandomRangeFloat(Minimum=550, Maximum=700)
     Alpha=$User.ParaGlideLeanAlpha
    )
   )
   &InitialMeshOrientation(Use Rotation Vector=false)
   /SetVariables(Particles.MeshIndex=&RandomRangeInt(Minimum=0, Maximum=1)) {enabled=false, isSetParameters=true}
  ]
  particleUpdateScript [
   &ParticleState()
   &SkeletalMeshLocation(
    Bone / Socket Index=&SelectIntFromArray(Direct Array Index=$Particles.PartitionParticles.Partition, Int Selection Array=DataInterface() {propertyValues {IntData [0,1]}})
    MaskSpawnGroup=false
    Mesh Sampling Type=::NewEnumerator5
    Orientation Sampling=::NewEnumerator1
    Position Sampling=::NewEnumerator1
    Sampled Position Offset:NiagaraPosition=$Particles.Initial.ShapeLocation.ShapePosition
    Sampled Position Offset Coordinate Space:ENiagaraSkelMeshTransforms=::NewEnumerator1
    Skeletal Mesh=$System.Skeletal Mesh
    Socket Sampling Mode=::NewEnumerator5
   )
   &UpdateMeshOrientation(
    Rotation Rate=&RandomRangeFloat(Minimum=-8, Maximum=8)
    Rotation Vector=&RandomRangeVector(Minimum=<-1 -1 -1>, Maximum=<1 1 1>)
   )
   &Drag(Drag=&RandomRangeFloat(Minimum=0.8, Maximum=1.2), Use Linear Drag=true, Use Rotational Drag=true)
   &SolveForcesAndVelocity(Clamp Velocity=false, Limit Acceleration=false)
   &ScaleColor(
    Scale Alpha=&FloatFromCurve(FloatCurve=DataInterface() {propertyValues {Curve {keys [
         {tangentMode=RCTM_User, time=0, value=0}
         <at 0.0599 1>
         {arriveTangent=-0.0405, interpMode=RCIM_Cubic, leaveTangent=-0.0405, tangentMode=RCTM_User, time=0.7009, value=1}
         {arriveTangent=-8.0089, interpMode=RCIM_Cubic, leaveTangent=-8.0089, tangentMode=RCTM_User, time=1, value=0}
        ]}}})
    Scale RGB=&MakeVectorFromLinearColorRGB(Color=&FNTimeOfDayColor(Night Color=<rgba 0.6302 0.7887 1 1>, Morning Color=<rgba 1 1 1 1>, Day Color=<rgba 1 1 1 1>, Evening Color=<rgba 1 1 1 1>))
    ScaleA=true
    ScaleRGB=true
   )
   &ScaleMeshSize(Scale Factor=&VectorFromFloat(Value=&FloatFromCurve(FloatCurve=DataInterface() {propertyValues {Curve {keys [
          <at 0 0 RCIM_Cubic>
          {interpMode=RCIM_Cubic, tangentMode=RCTM_User, time=0.2169, value=1}
          {interpMode=RCIM_Cubic, tangentMode=RCTM_User, time=0.5, value=1}
          {arriveTangent=-1.4032, interpMode=RCIM_Cubic, leaveTangent=-1.4032, tangentMode=RCTM_User, time=1, value=0.5}
         ]}}})))
   &ScaleMeshSize(Scale Factor=&VectorFromFloat(Value=&Lerp_Float(
      A=1
      B=0.925
      Alpha=&Multiply_Float(
       A=&DotProduct(
        Vector A=&NormalizeVector(
         Vector To Normalize=&ReturnCameraVector(Camera Vectors=::NewEnumerator1, Coordinate Space=::World)
         Safe Normalize=true
         Safe Fallback Threshold=0
         Safe Fallback Vector=<0 0 0.001>
        )
        Vector B=&NormalizeVector(Vector To Normalize=$Engine.Owner.SystemZAxis, Safe Normalize=true, Safe Fallback Threshold=0, Safe Fallback Vector=<0 0 0.001>)
       )
       B=-1
      )
     ))) {module=ScaleMeshSize001}
   /SetVariables(Particles.Tomato Color:LinearColor=&MakeLinearColorFromVectorAndFloat(
     "Vector (RGB)"=&MakeVectorFromLinearColorRGB(Color=&FNTimeOfDayColor(Night Color=<rgba 0.1834 0.2687 0.8385 1>, Morning Color=<rgba 1 1 1 1>, Day Color=<rgba 1 1 1 1>, Evening Color=<rgba 1 1 1 1>))
     "Float (Alpha)"=&FloatFromCurve(FloatCurve=DataInterface() {propertyValues {Curve {keys [
          {tangentMode=RCTM_User, time=0, value=0}
          <at 0.0599 1>
          {arriveTangent=-0.0405, interpMode=RCIM_Cubic, leaveTangent=-0.0405, tangentMode=RCTM_User, time=0.7009, value=1}
          {arriveTangent=-8.0089, interpMode=RCIM_Cubic, leaveTangent=-8.0089, tangentMode=RCTM_User, time=1, value=0}
         ]}}})
    )) {isSetParameters=true}
  ]
  rendererClasses [&NiagaraMeshRendererProperties]
  renderers [
   {rendererClass=&NiagaraMeshRendererProperties, rendererIndex=0}
   {rendererClass=&NiagaraMeshRendererProperties, rendererIndex=1}
   {rendererClass=&NiagaraMeshRendererProperties, rendererIndex=2}
  ]
 }
 GreaseMesh {
  emitterUpdateScript [
   &EmitterState(Life Cycle Mode=::NewEnumerator1, Loop Behavior=::NewEnumerator1, Loop Duration Mode=::NewEnumerator1, UseLoopDelay=false)
   &SpawnBurst_Instantaneous(Spawn Count=2, Use Spawn Probability=false, UseLoopCountLimit=false)
  ]
  particleSpawnScript [
   &PartitionParticles()
   &InitializeParticle(
    Color=<rgba 0.2789 0.0723 0.0296 1>
    Lifetime=5
    Mesh Scale Mode=::NewEnumerator3
    Mesh Uniform Scale=0.2
    UsePositionOffset=false
    Write Material Random=true
   )
   &InitialMeshOrientation(
    Mesh Orientation Mode=::NewEnumerator1
    Rotation=&MaskVector(Mask Source=1, Mask Parameter=$Particles.PartitionParticles.Partition, Mask=0, Masked Vector=<0 0.04 0>, Passthrough Vector=<0 -0.04 0>)
    Use Rotation Vector=true
   )
  ]
  particleUpdateScript [
   &ParticleState(Kill Particles When Lifetime Has Elapsed=false, Let Infinitely Lived Particles Die When Emitter Deactivates=true)
   &SkeletalMeshLocation(
    Bone / Socket Index=&SelectIntFromArray(Direct Array Index=$Particles.PartitionParticles.Partition, Int Selection Array=DataInterface() {propertyValues {IntData [0,1]}})
    MaskSpawnGroup=false
    Mesh Sampling Type=::NewEnumerator5
    Orientation Sampling=::NewEnumerator1
    Sampled Position Offset=&AddVectorToPosition(
     Position=&MaskVector(
      Mask Source=1
      Mask Parameter=$Particles.PartitionParticles.Partition
      Mask=0
      Masked Vector=<9.9298 0 36.5>
      Passthrough Vector=<-9.9298 1 36.5>
     )
     Vector=&MakeVector(
      X=&Sine(Normalized Angle=$Emitter.Age, Period=&RandomRangeFloat(Minimum=0.8, Maximum=1.2), Scale=0.25, Bias=0)
      Y=&Sine(Normalized Angle=$Emitter.Age, Period=&RandomRangeFloat(Minimum=0.8, Maximum=1.2), Scale=0.25, Bias=0)
      Z=&Sine(Normalized Angle=$Emitter.Age, Period=&RandomRangeFloat(Minimum=0.8, Maximum=1.2), Scale=1, Bias=0)
     )
    )
    Sampled Position Offset Coordinate Space:ENiagaraSkelMeshTransforms=::NewEnumerator1
    Skeletal Mesh=$System.Skeletal Mesh
    Socket Sampling Mode=::NewEnumerator5
   )
   &DynamicMaterialParameters(
    Index 0 Param 2=&Waveform("Waveform [1]"=::NewEnumerator2, "[1] Frequency"=1, "[1] Amplitude Min/Max"=<-13 -4>)
    Index 0 Param 3=&Add_Float(A=0, B=&Multiply_Float(A=$Engine.Time, B=0.4))
    Module.Param0WriteEnabled=true
    Module.Param0WriteEnabled1=true
    Module.Param1WriteEnabled=true
    Module.Param1WriteEnabled1=true
    Module.Param2WriteEnabled=true
    Module.Param2WriteEnabled1=true
    Module.Param3WriteEnabled=true
    Module.Param3WriteEnabled1=true
    Write Parameter Index 1=true
   )
   &ScaleColor(
    Scale Alpha=&Multiply_Float(
     A=&Lerp_Float(
      A=0
      B=1
      Alpha=&RemapRange(Input Value To Remap=$User.ParaGlideLeanAlpha, Input Min=0.16, Input Max=1, Output Min=0, Output Max=1)
     )
     B=&Waveform("Waveform [1]"=::NewEnumerator2, "[1] Frequency"=1, "[1] Amplitude Min/Max"=<0.725 0.925>)
    )
    ScaleA=true
    ScaleRGB=true
   )
  ]
  rendererClasses [&NiagaraMeshRendererProperties]
  renderers [{rendererClass=&NiagaraMeshRendererProperties, rendererIndex=0}]
 }
 KetchupMesh {
  emitterUpdateScript [
   &EmitterState(Life Cycle Mode=::NewEnumerator1, Loop Behavior=::NewEnumerator1, Loop Duration Mode=::NewEnumerator1, UseLoopDelay=false)
   &SpawnRate(SpawnRate=&Lerp_Float(A=12, B=16, Alpha=$User.ParaGlideLeanAlpha), Use Spawn Probability=false)
   /SetVariables(Emitter.UVTiling:Vector4f=&MakeVector4(
     X=&Lerp_Float(
      A=&RandomRangeFloat(Minimum=0.7, Maximum=1)
      B=&RandomRangeFloat(Minimum=1, Maximum=1.45)
      Alpha=$User.ParaGlideLeanAlpha
     )
     Y=&RandomRangeFloat(Minimum=0.7, Maximum=1)
     Z=1
     W=1
    )) {isSetParameters=true, module=SetVariables_#2}
  ]
  particleSpawnScript [
   &PartitionParticles()
   &InitializeParticle(
    Color=<rgba 1 0 0 1>
    Lifetime=&Lerp_Float(A=1.2, B=0.6, Alpha=$User.ParaGlideLeanAlpha)
    Mesh Scale Mode=::NewEnumerator3
    UsePositionOffset=false
    Write Material Random=true
   )
   &InitialMeshOrientation(Mesh Orientation Mode=::NewEnumerator1, Use Rotation Vector=true)
   &InitialMeshOrientation(
    Mesh Orientation Mode=::NewEnumerator1
    Rotation=&MaskVector(
     Mask Source=1
     Mask Parameter=$Particles.PartitionParticles.Partition
     Mask=0
     Masked Vector=&Lerp_Vector(A=<0 0 0.3>, B=<-0.033 0 0.3>, Alpha=$User.ParaGlideLeanAlpha)
     Passthrough Vector=&Lerp_Vector(A=<0 0 0>, B=<0 0.02 0>, Alpha=$User.ParaGlideLeanAlpha)
    )
    Use Rotation Vector=true
   ) {module=InitialMeshOrientation001}
  ]
  particleUpdateScript [
   &ParticleState()
   &ScaleMeshSize(Scale Factor=&Lerp_Vector(
     A=&MergeHelper_VectorSwitch(0=<1.6 1.6 0.1>, 1=<1.6 1.6 0.075>, 2=?, 3=?, Index=$Particles.PartitionParticles.Partition, Num=::NewEnumerator0)
     B=&MergeHelper_VectorSwitch(0=<1.6 1.6 0.25>, 1=<1.6 1.6 0.15>, 2=?, 3=?, Index=$Particles.PartitionParticles.Partition, Num=::NewEnumerator0)
     Alpha=$User.ParaGlideLeanAlpha
    ))
   &SkeletalMeshLocation(
    Bone / Socket Index=&SelectIntFromArray(Direct Array Index=$Particles.PartitionParticles.Partition, Int Selection Array=DataInterface() {propertyValues {IntData [0,1]}})
    MaskSpawnGroup=false
    Mesh Sampling Type=::NewEnumerator5
    Orientation Sampling=::NewEnumerator1
    Sampled Position Offset=&AddVectorToPosition(
     Position=&Lerp_Vector(
      A=&MergeHelper_VectorSwitch(0=<0 0 25>, 1=<0 0 25>, 2=?, 3=?, Index=$Particles.PartitionParticles.Partition, Num=::NewEnumerator0)
      B=&MergeHelper_VectorSwitch(0=<0 0 66>, 1=<0 0 66>, 2=?, 3=?, Index=$Particles.PartitionParticles.Partition, Num=::NewEnumerator0)
      Alpha=$User.ParaGlideLeanAlpha
     )
     Vector=<0 0 0>
    )
    Sampled Position Offset Coordinate Space:ENiagaraSkelMeshTransforms=::NewEnumerator1
    Skeletal Mesh=$System.Skeletal Mesh
    Socket Sampling Mode=::NewEnumerator5
   )
   &ScaleMeshSize(Scale Factor=&Multiply_Vector(
     A=&VectorFromCurve(VectorCurve=DataInterface() {propertyValues {
        XCurve {defaultValue=0.2, keys [{arriveTangent=3.0023, time=0, value=0.6}, <at 1 1>]}
        YCurve {keys [{arriveTangent=0.803, time=0, value=0.6}, {arriveTangent=0.465, leaveTangent=0.465, time=1, value=0.8}]}
        ZCurve {
         defaultValue=0.2
         keys [
          {arriveTangent=3.0023, interpMode=RCIM_Cubic, leaveTangent=3.0023, tangentMode=RCTM_User, time=0, value=0.3}
          {arriveTangentWeight=0.1667, interpMode=RCIM_Cubic, tangentMode=RCTM_User, time=0.5, value=1}
          {arriveTangent=-5.505, arriveTangentWeight=0.1667, interpMode=RCIM_Cubic, leaveTangent=-5.505, tangentMode=RCTM_User, time=1, value=0}
         ]
        }
       }})
     B=&VectorFromFloat(Value=&FloatFromCurve(
       FloatCurve=DataInterface() {propertyValues {Curve {
          defaultValue=1
          keys [{arriveTangent=3.0116, interpMode=RCIM_Cubic, time=0, value=0}, {interpMode=RCIM_Cubic, leaveTangent=0.4934, time=0.75, value=1}]
         }}}
       CurveIndex=$Emitter.Age
      ))
    )) {module=ScaleMeshSize001}
   &ScaleColor(
    Scale Alpha=&FloatFromCurve(
     FloatCurve=DataInterface() {propertyValues {Curve {keys [
         {arriveTangent=-0.6531, interpMode=RCIM_Cubic, time=0, value=1}
         {arriveTangent=-0.8638, interpMode=RCIM_Cubic, leaveTangent=-0.8638, tangentMode=RCTM_User, time=0.8, value=0.3}
         {arriveTangent=-0.4827, interpMode=RCIM_Cubic, leaveTangent=-0.4827, tangentMode=RCTM_User, time=1, value=0.1}
        ]}}}
     Scale Curve=&Lerp_Float(
      A=0
      B=1
      Alpha=&RemapRange(Input Value To Remap=$User.ParaGlideLeanAlpha, Input Min=0.16, Input Max=1, Output Min=0, Output Max=1)
     )
    )
    Scale RGB=&MakeVectorFromLinearColorRGB(Color=&FNTimeOfDayColor(Night Color=<rgba 0.4271 0.5366 1 1>, Morning Color=<rgba 1 1 1 1>, Day Color=<rgba 1 1 1 1>, Evening Color=<rgba 1 1 1 1>))
    ScaleA=true
    ScaleRGB=true
   )
   &DynamicMaterialParameters(
    Index 0 Param 2=&Lerp_Float(A=0, B=3, Alpha=$User.ParaGlideLeanAlpha)
    Index 0 Param 3=&RandomRangeFloat(Minimum=0, Maximum=1)
    Module.Index0Float0Write=true
    Module.Index0Float1Write=true
    Module.Index0Float2Write=true
    Module.Index0Float3Write=true
   )
  ]
  rendererClasses [&NiagaraMeshRendererProperties]
  renderers [{rendererClass=&NiagaraMeshRendererProperties, rendererIndex=0}]
 }
 KetchupMustardBits {
  emitterUpdateScript [
   &EmitterState(Life Cycle Mode=::NewEnumerator1, Loop Behavior=::NewEnumerator1, Loop Duration Mode=::NewEnumerator1, UseLoopDelay=false)
   &SpawnRate(SpawnRate=&Lerp_Float(A=32, B=64, Alpha=$User.ParaGlideLeanAlpha), Use Spawn Probability=false)
   &SpawnRate(SpawnGroup=2, SpawnRate=&Lerp_Float(A=24, B=32, Alpha=$User.ParaGlideLeanAlpha), Use Spawn Probability=false) {module=SpawnRate001}
   &SpawnRate(SpawnGroup=1, SpawnRate=&Lerp_Float(A=24, B=32, Alpha=$User.ParaGlideLeanAlpha), Use Spawn Probability=false) {module=SpawnRate002}
   /SetVariables(
    Emitter.Grease 1:LinearColor=<rgba 0.2159 0.0529 0.0194 1>
    Emitter.Grease 2:LinearColor=<rgba 0.2159 0.0529 0.0194 1>
    Emitter.Ketchup 1:LinearColor=<rgba 0.1448 0.0058 0.0024 1>
    Emitter.Ketchup 2:LinearColor=<rgba 1 0 0 1>
    Emitter.Mustard 2:LinearColor=<rgba 1 0.59 0 1>
    Emitter.mustard 1:LinearColor=<rgba 0.5323 0.3141 0 1>
   ) {isSetParameters=true}
  ]
  particleSpawnScript [
   /SetVariables(Particles.VisibilityTag=&SpawnGroup()) {isSetParameters=true}
   &PartitionParticles(Partition Mode:ENiagara_PartitionMode=::NewEnumerator1)
   &InitializeParticle(
    Color Mode=::NewEnumerator0
    Lifetime=&MergeHelper_FloatSwitch(
     0=&RandomRangeFloat(Minimum=0.3, Maximum=1)
     1=0.2
     2=&RandomRangeFloat(Minimum=0.3, Maximum=1)
     3=?
     Index=&SpawnGroup()
     Num=::NewEnumerator1
    )
    UsePositionOffset=false
    Write Material Random=true
   )
   &ShapeLocation(
    Cylinder Height=&MergeHelper_FloatSwitch(0=20, 1=20, 2=4, 3=?, Index=&SpawnGroup(), Num=::NewEnumerator1)
    Cylinder Radius=&MergeHelper_FloatSwitch(0=4, 1=10, 2=12, 3=?, Index=&SpawnGroup(), Num=::NewEnumerator1)
    MaskSpawnGroup=false
    Offset Mode=::NewEnumerator0
    Override Seed=false
    Shape Primitive=::NewEnumerator1
    Surface Only=false
    UseOwnerScale=true
   )
   &AddVelocity(
    Cone Angle=15
    Cone Axis=&Multiply_VectorByFloat(Vector=$System.FN_Contrail_Setup.Direction Vector, Float=-1)
    Module.UseConeFalloff=false
    Module.UseVelDistribution=false
    Override Seed=false
    Velocity Mode=::NewEnumerator2
    Velocity Speed=&MergeHelper_FloatSwitch(
     0=&Lerp_Float(A=225, B=550, Alpha=$User.ParaGlideLeanAlpha)
     1=0
     2=&Lerp_Float(A=120, B=200, Alpha=$User.ParaGlideLeanAlpha)
     3=?
     Index=&SpawnGroup()
     Num=::NewEnumerator1
    )
   )
   &AddVelocity(Velocity=&MakeVector(
     X=0
     Y=0
     Z=&MergeHelper_FloatSwitch(
      0=0
      1=&Lerp_Float(A=120, B=325, Alpha=$User.ParaGlideLeanAlpha)
      2=0
      3=0
      Index=&SpawnGroup()
      Num=::NewEnumerator2
     )
    )) {module=AddVelocity001}
  ]
  particleUpdateScript [
   &ParticleState()
   &SkeletalMeshLocation(
    Bone / Socket Index=&SelectIntFromArray(Direct Array Index=$Particles.PartitionParticles.Partition, Int Selection Array=DataInterface() {propertyValues {IntData [0,1]}})
    MaskSpawnGroup=false
    Mesh Sampling Type=::NewEnumerator5
    Orientation Sampling=::NewEnumerator1
    Position Sampling=::NewEnumerator1
    Sampled Position Offset=&AddVectorToPosition(
     Position=$Particles.Initial.ShapeLocation.ShapePosition
     Vector=&MergeHelper_VectorSwitch(
      0=<0 0 0>
      1=&Lerp_Vector(
       A=&MaskVector(Mask Source=1, Mask Parameter=$Particles.PartitionParticles.Partition, Mask=0, Masked Vector=<0 0 20>, Passthrough Vector=<0 0 20>)
       B=&MaskVector(
        Mask Source=1
        Mask Parameter=$Particles.PartitionParticles.Partition
        Mask=0
        Masked Vector=<8 -1.85 107>
        Passthrough Vector=<-8 -1.85 107>
       )
       Alpha=$User.ParaGlideLeanAlpha
      )
      2=&MaskVector(Mask Source=1, Mask Parameter=$Particles.PartitionParticles.Partition, Mask=0, Masked Vector=<8 0 12>, Passthrough Vector=<-8 0 12>)
      3=?
      Index=&SpawnGroup()
      Num=::NewEnumerator1
     )
    )
    Sampled Position Offset Coordinate Space:ENiagaraSkelMeshTransforms=::NewEnumerator1
    Skeletal Mesh=$System.Skeletal Mesh
    Socket Sampling Mode=::NewEnumerator5
   )
   /SetVariables(Particles.KetchupMustardScale:Vector3f=&Multiply_Vector(
     A=&RandomRangeVector(Minimum=<1 0.25 0.2>, Maximum=<4 2 0.5>)
     B=&VectorFromFloat(Value=&FloatFromCurve(
       FloatCurve=DataInterface() {propertyValues {Curve {keys [
           <at 0 0 RCIM_Cubic>
           {interpMode=RCIM_Cubic, tangentMode=RCTM_User, time=0.2169, value=1}
           {interpMode=RCIM_Cubic, tangentMode=RCTM_User, time=0.5, value=1}
           {arriveTangent=-1.4032, interpMode=RCIM_Cubic, leaveTangent=-1.4032, tangentMode=RCTM_User, time=1, value=0.5}
          ]}}}
       Scale Curve=&RandomRangeFloat(Minimum=0.4, Maximum=0.65)
      ))
    )) {isSetParameters=true}
   /SetVariables(Particles.GreaseSCale:Vector3f=&Multiply_Vector(
     A=&RandomRangeVector(Minimum=<1 0.25 0.2>, Maximum=<4 2 0.5>)
     B=&VectorFromFloat(Value=&FloatFromCurve(
       FloatCurve=DataInterface() {propertyValues {Curve {keys [
           <at 0 0 RCIM_Cubic>
           {interpMode=RCIM_Cubic, tangentMode=RCTM_User, time=0.2169, value=1}
           {interpMode=RCIM_Cubic, tangentMode=RCTM_User, time=0.5, value=1}
           {arriveTangent=-1.4032, interpMode=RCIM_Cubic, leaveTangent=-1.4032, tangentMode=RCTM_User, time=1, value=0.5}
          ]}}}
       Scale Curve=&RandomRangeFloat(Minimum=0.375, Maximum=0.5)
      ))
    )) {isSetParameters=true}
   &Drag(Drag=&RandomRangeFloat(Minimum=0.8, Maximum=1.2), Use Linear Drag=true, Use Rotational Drag=true)
   &CurlNoiseForce(
    Bias Noise=false
    Noise Strength=&MergeHelper_FloatSwitch(
     0=&RandomRangeFloat(Minimum=120, Maximum=225)
     1=0
     2=&RandomRangeFloat(Minimum=120, Maximum=225)
     3=?
     Index=&SpawnGroup()
     Num=::NewEnumerator1
    )
    Pan Noise=false
   )
   &SolveForcesAndVelocity(Clamp Velocity=false, Limit Acceleration=false)
   &ScaleColor(
    Scale Alpha=&FloatFromCurve(FloatCurve=DataInterface() {propertyValues {Curve {keys [
         {tangentMode=RCTM_User, time=0, value=0}
         <at 0.0599 1>
         {arriveTangent=-0.0405, interpMode=RCIM_Cubic, leaveTangent=-0.0405, tangentMode=RCTM_User, time=0.7009, value=1}
         {arriveTangent=-8.0089, interpMode=RCIM_Cubic, leaveTangent=-8.0089, tangentMode=RCTM_User, time=1, value=0}
        ]}}})
    Scale RGB=&MakeVectorFromLinearColorRGB(Color=&FNTimeOfDayColor(Night Color=<rgba 0.6302 0.7887 1 1>, Morning Color=<rgba 1 1 1 1>, Day Color=<rgba 1 1 1 1>, Evening Color=<rgba 1 1 1 1>))
    ScaleA=true
    ScaleRGB=true
   )
   /SetVariables(
    Particles.MotionLineColor:LinearColor=&MakeLinearColorFromVectorAndFloat(
     "Vector (RGB)"=&MakeVectorFromLinearColorRGB(Color=<rgba 0.624 0.2016 0.0242 1>)
     "Float (Alpha)"=&FloatFromCurve(FloatCurve=DataInterface() {propertyValues {Curve {
         defaultValue=1
         keys [
          {interpMode=RCIM_Cubic, tangentMode=RCTM_User, time=0, value=0}
          {interpMode=RCIM_Cubic, tangentMode=RCTM_User, time=0.5, value=1}
          {tangentMode=RCTM_User, time=1, value=0}
         ]
        }}})
    )
    Particles.MotionLineSize:Vector2f=&Vector2DFromCurve(
     Vector2Curve=DataInterface() {propertyValues {
       XCurve {keys [{tangentMode=RCTM_User, time=0, value=0.5}, {tangentMode=RCTM_User, time=1, value=1.5}]}
       YCurve {keys [{tangentMode=RCTM_User, time=0, value=1}, {tangentMode=RCTM_User, time=1, value=0.5}]}
      }}
     ScaleCurve=&Lerp_Vector2(A=<45 80>, B=<45 120>, Alpha=$User.ParaGlideLeanAlpha)
    )
   ) {isSetParameters=true}
  ]
  rendererClasses [&NiagaraMeshRendererProperties, &NiagaraSpriteRendererProperties]
  renderers [
   {rendererClass=&NiagaraMeshRendererProperties, rendererIndex=0}
   {rendererClass=&NiagaraMeshRendererProperties, rendererIndex=1}
   {rendererClass=&NiagaraSpriteRendererProperties, rendererIndex=2}
   {rendererClass=&NiagaraMeshRendererProperties, rendererIndex=3}
  ]
 }
 Mesh_SkeletalMesh {
  emitterUpdateScript [
   &EmitterState(Life Cycle Mode=::NewEnumerator1, Loop Behavior=::NewEnumerator1, Loop Duration Mode=::NewEnumerator1, UseLoopDelay=false)
   &SpawnBurst_Instantaneous(Spawn Count=2, Spawn Probability=HlslExpression() {hlslExpression="Emitter.LoopCount == 0"}, Use Spawn Probability=false)
  ]
  particleSpawnScript [
   &PartitionParticles()
   &InitializeParticle(Lifetime=5, Sprite Size=<10 100>, Sprite Size Mode=::NewEnumerator4, UsePositionOffset=false, Write Material Random=true)
   &InitialMeshOrientation(
    Mesh Orientation Mode=::NewEnumerator1
    Rotation=&MakeVector(X=0, Y=-0.25, Z=0)
    Use Rotation Vector=true
   )
  ]
  particleUpdateScript [
   &ParticleState(Kill Particles When Lifetime Has Elapsed=false, Let Infinitely Lived Particles Die When Emitter Deactivates=true)
   &SkeletalMeshLocation(
    Bone / Socket Index=&SelectIntFromArray(Direct Array Index=$Particles.PartitionParticles.Partition, Int Selection Array=DataInterface() {propertyValues {IntData [0,1]}})
    MaskSpawnGroup=false
    Mesh Sampling Type=::NewEnumerator5
    Orientation Sampling=::NewEnumerator1
    Sampled Position Offset Coordinate Space:ENiagaraSkelMeshTransforms=::NewEnumerator1
    Skeletal Mesh=$System.Skeletal Mesh
    Socket Sampling Mode=::NewEnumerator5
   )
   /SetVariables(
    Particles.Bun_Bottom_Pos:NiagaraPosition=&AddVectorToPosition(
     Position=$Particles.SkeletalMeshLocation.SampledPosition
     Vector=&AddVectorToPosition(
      Position=<0 0 12>
      Vector=&MakeVector(
       X=&Sine(Normalized Angle=$Emitter.Age, Period=&RandomRangeFloat(Minimum=0.8, Maximum=1.2), Scale=0.25, Bias=0)
       Y=&Sine(Normalized Angle=$Emitter.Age, Period=&RandomRangeFloat(Minimum=0.8, Maximum=1.2), Scale=0.25, Bias=0)
       Z=&Sine(Normalized Angle=$Emitter.Age, Period=&RandomRangeFloat(Minimum=0.8, Maximum=1.2), Scale=1, Bias=0)
      )
     )
    )
    Particles.Bun_Top_Pos:NiagaraPosition=&AddVectorToPosition(
     Position=$Particles.SkeletalMeshLocation.SampledPosition
     Vector=&AddVectorToPosition(
      Position=<0 0 28>
      Vector=&MakeVector(
       X=&Sine(Normalized Angle=$Emitter.Age, Period=&RandomRangeFloat(Minimum=0.8, Maximum=1.2), Scale=0.25, Bias=0)
       Y=&Sine(Normalized Angle=$Emitter.Age, Period=&RandomRangeFloat(Minimum=0.8, Maximum=1.2), Scale=0.25, Bias=0)
       Z=&Sine(Normalized Angle=$Emitter.Age, Period=&RandomRangeFloat(Minimum=0.8, Maximum=1.2), Scale=0.2, Bias=0)
      )
     )
    )
    Particles.Cheese_Pos:NiagaraPosition=&AddVectorToPosition(
     Position=$Particles.SkeletalMeshLocation.SampledPosition
     Vector=&AddVectorToPosition(
      Position=<0 0 17.8>
      Vector=&MakeVector(
       X=&Sine(Normalized Angle=$Emitter.Age, Period=&RandomRangeFloat(Minimum=0.8, Maximum=1.2), Scale=0.25, Bias=0)
       Y=&Sine(Normalized Angle=$Emitter.Age, Period=&RandomRangeFloat(Minimum=0.8, Maximum=1.2), Scale=0.25, Bias=0)
       Z=&Sine(Normalized Angle=$Emitter.Age, Period=&RandomRangeFloat(Minimum=0.8, Maximum=1.2), Scale=1, Bias=0)
      )
     )
    )
    Particles.Lettuce_Pos:NiagaraPosition=&AddVectorToPosition(
     Position=$Particles.SkeletalMeshLocation.SampledPosition
     Vector=&AddVectorToPosition(
      Position=<0 0 21.0355>
      Vector=&MakeVector(
       X=&Sine(Normalized Angle=$Emitter.Age, Period=&RandomRangeFloat(Minimum=0.8, Maximum=1.2), Scale=0.25, Bias=0)
       Y=&Sine(Normalized Angle=$Emitter.Age, Period=&RandomRangeFloat(Minimum=0.8, Maximum=1.2), Scale=0.25, Bias=0)
       Z=&Sine(Normalized Angle=$Emitter.Age, Period=&RandomRangeFloat(Minimum=0.8, Maximum=1.2), Scale=1, Bias=0)
      )
     )
    )
    Particles.Patty_Pos:NiagaraPosition=&AddVectorToPosition(
     Position=$Particles.SkeletalMeshLocation.SampledPosition
     Vector=&AddVectorToPosition(
      Position=<0 0 13>
      Vector=&MakeVector(
       X=&Sine(Normalized Angle=$Emitter.Age, Period=&RandomRangeFloat(Minimum=0.8, Maximum=1.2), Scale=0.25, Bias=0)
       Y=&Sine(Normalized Angle=$Emitter.Age, Period=&RandomRangeFloat(Minimum=0.8, Maximum=1.2), Scale=0.25, Bias=0)
       Z=&Sine(Normalized Angle=$Emitter.Age, Period=&RandomRangeFloat(Minimum=0.8, Maximum=1.2), Scale=1, Bias=0)
      )
     )
    )
    Particles.Pickle_Pos:NiagaraPosition=&AddVectorToPosition(
     Position=$Particles.SkeletalMeshLocation.SampledPosition
     Vector=&AddVectorToPosition(
      Position=<0 0 27.581>
      Vector=&MakeVector(
       X=&Sine(Normalized Angle=$Emitter.Age, Period=&RandomRangeFloat(Minimum=0.8, Maximum=1.2), Scale=0.25, Bias=0)
       Y=&Sine(Normalized Angle=$Emitter.Age, Period=&RandomRangeFloat(Minimum=0.8, Maximum=1.2), Scale=0.25, Bias=0)
       Z=&Sine(Normalized Angle=$Emitter.Age, Period=&RandomRangeFloat(Minimum=0.8, Maximum=1.2), Scale=1, Bias=0)
      )
     )
    )
    Particles.Tomato_Pos:NiagaraPosition=&AddVectorToPosition(
     Position=$Particles.SkeletalMeshLocation.SampledPosition
     Vector=&AddVectorToPosition(
      Position=<0 0 24.867>
      Vector=&MakeVector(
       X=&Sine(Normalized Angle=$Emitter.Age, Period=&RandomRangeFloat(Minimum=0.8, Maximum=1.2), Scale=0.25, Bias=0)
       Y=&Sine(Normalized Angle=$Emitter.Age, Period=&RandomRangeFloat(Minimum=0.8, Maximum=1.2), Scale=0.25, Bias=0)
       Z=&Sine(Normalized Angle=$Emitter.Age, Period=&RandomRangeFloat(Minimum=0.8, Maximum=1.2), Scale=0.2, Bias=0)
      )
     )
    )
   ) {isSetParameters=true}
   /SetVariables(
    Particles.Bun_Bottom_Pos_Update:NiagaraPosition=&Lerp_Vector(
     A=Linked() {linkedVariable {name=Particles.Bun_Bottom_Pos, type=NiagaraPosition}}
     B=&AddVectorToPosition(
      Position=$Particles.Bun_Bottom_Pos
      Vector=&AddVectorToPosition(
       Position=<0 0 10>
       Vector=&MergeHelper_VectorSwitch(0=<0 0 0>, 1=<0 0 0>, 2=?, 3=?, Index=$Particles.PartitionParticles.Partition, Num=::NewEnumerator0)
      )
     )
     Alpha=$User.ParaGlideLeanAlpha
    )
    Particles.Bun_Top_Pos_Update:NiagaraPosition=&Lerp_Vector(
     A=Linked() {linkedVariable {name=Particles.Bun_Top_Pos, type=NiagaraPosition}}
     B=&AddVectorToPosition(
      Position=$Particles.Bun_Top_Pos
      Vector=&AddVectorToPosition(
       Position=<0 0 60>
       Vector=&MergeHelper_VectorSwitch(0=<8 0 0>, 1=<-8 0 0>, 2=?, 3=?, Index=$Particles.PartitionParticles.Partition, Num=::NewEnumerator0)
      )
     )
     Alpha=$User.ParaGlideLeanAlpha
    )
    Particles.Cheese_Pos_Update:NiagaraPosition=&Lerp_Vector(
     A=Linked() {linkedVariable {name=Particles.Cheese_Pos, type=NiagaraPosition}}
     B=&AddVectorToPosition(
      Position=$Particles.Cheese_Pos
      Vector=&AddVectorToPosition(
       Position=<0 0 32>
       Vector=&MergeHelper_VectorSwitch(0=<10.5 0 0>, 1=<-10.5 0 0>, 2=?, 3=?, Index=$Particles.PartitionParticles.Partition, Num=::NewEnumerator0)
      )
     )
     Alpha=$User.ParaGlideLeanAlpha
    )
    Particles.Lettuce_Pos_Update:NiagaraPosition=&Lerp_Vector(
     A=Linked() {linkedVariable {name=Particles.Lettuce_Pos, type=NiagaraPosition}}
     B=&AddVectorToPosition(
      Position=$Particles.Lettuce_Pos
      Vector=&AddVectorToPosition(
       Position=<0 0 36.1525>
       Vector=&MergeHelper_VectorSwitch(0=<3 0 0>, 1=<-3 0 0>, 2=?, 3=?, Index=$Particles.PartitionParticles.Partition, Num=::NewEnumerator0)
      )
     )
     Alpha=$User.ParaGlideLeanAlpha
    )
    Particles.Patty_Pos_Update:NiagaraPosition=&Lerp_Vector(
     A=Linked() {linkedVariable {name=Particles.Patty_Pos, type=NiagaraPosition}}
     B=&AddVectorToPosition(
      Position=$Particles.Patty_Pos
      Vector=&AddVectorToPosition(
       Position=<0 0 14>
       Vector=&MergeHelper_VectorSwitch(0=<8 0 0>, 1=<-8 0 0>, 2=?, 3=?, Index=$Particles.PartitionParticles.Partition, Num=::NewEnumerator0)
      )
     )
     Alpha=$User.ParaGlideLeanAlpha
    )
    Particles.Pickle_Pos_Update:NiagaraPosition=&Lerp_Vector(
     A=Linked() {linkedVariable {name=Particles.Pickle_Pos, type=NiagaraPosition}}
     B=&AddVectorToPosition(
      Position=$Particles.Pickle_Pos
      Vector=&AddVectorToPosition(
       Position=<0 0 52>
       Vector=&MergeHelper_VectorSwitch(0=<14.5649 0 0>, 1=<-14.5649 0 0>, 2=?, 3=?, Index=$Particles.PartitionParticles.Partition, Num=::NewEnumerator0)
      )
     )
     Alpha=$User.ParaGlideLeanAlpha
    )
    Particles.Tomato_Pos_Update:NiagaraPosition=&Lerp_Vector(
     A=Linked() {linkedVariable {name=Particles.Tomato_Pos, type=NiagaraPosition}}
     B=&AddVectorToPosition(
      Position=$Particles.Tomato_Pos
      Vector=&AddVectorToPosition(
       Position=<0 0 42>
       Vector=&MergeHelper_VectorSwitch(0=<0.5 0 0>, 1=<-0.5 0 0>, 2=?, 3=?, Index=$Particles.PartitionParticles.Partition, Num=::NewEnumerator0)
      )
     )
     Alpha=$User.ParaGlideLeanAlpha
    )
   ) {isSetParameters=true}
   /SetVariables(
    Particles.Bun_TopOrientation:Quat4f=&LerpQuaternion(
     Quaternion A:Quat4f=&MultiplyQuaternion(
      Quaternion A=&MakeQuaternion(
       Quaternion From=0
       Axis=<0 1 0>
       Angle=0
       From=?
       To=?
       Stabilization Mode:ENiagara_OrientVectorCount=::NewEnumerator0
       Side Direction:Vector3f=?
       Up Direction:Vector3f=?
       Angle Type=::NewEnumerator0
       Normalize:NiagaraBool=true
      )
      Quaternion B=&MultiplyQuaternion(
       Quaternion A=&MakeQuaternion(
        Quaternion From=0
        Axis=<0 1 0>
        Angle=&MergeHelper_FloatSwitch(0=2, 1=-2, 2=?, 3=?, Index=$Particles.PartitionParticles.Partition, Num=::NewEnumerator0)
        From=?
        To=?
        Stabilization Mode:ENiagara_OrientVectorCount=::NewEnumerator0
        Side Direction:Vector3f=?
        Up Direction:Vector3f=?
        Angle Type=::NewEnumerator0
        Normalize:NiagaraBool=true
       )
       Quaternion B=&MakeQuaternion(
        Quaternion From=0
        Axis=<0 0 1>
        Angle=&MergeHelper_FloatSwitch(
         0=&Multiply_Float(A=$Emitter.Age, B=-150)
         1=&Multiply_Float(A=$Emitter.Age, B=150)
         2=?
         3=?
         Index=$Particles.PartitionParticles.Partition
         Num=::NewEnumerator0
        )
        From=?
        To=?
        Stabilization Mode:ENiagara_OrientVectorCount=::NewEnumerator0
        Side Direction:Vector3f=?
        Up Direction:Vector3f=?
        Angle Type=::NewEnumerator0
        Normalize:NiagaraBool=true
       )
      )
     )
     Quaternion B:Quat4f=&MultiplyQuaternion(
      Quaternion A=&MakeQuaternion(
       Quaternion From=0
       Axis=<0 1 0>
       Angle=&MergeHelper_FloatSwitch(0=4, 1=-4, 2=?, 3=?, Index=$Particles.PartitionParticles.Partition, Num=::NewEnumerator0)
       From=?
       To=?
       Stabilization Mode:ENiagara_OrientVectorCount=::NewEnumerator0
       Side Direction:Vector3f=?
       Up Direction:Vector3f=?
       Angle Type=::NewEnumerator0
       Normalize:NiagaraBool=true
      )
      Quaternion B=&MakeQuaternion(
       Quaternion From=0
       Axis=<0 0 1>
       Angle=&MergeHelper_FloatSwitch(
        0=&Multiply_Float(A=$Emitter.Age, B=250)
        1=&Multiply_Float(A=$Emitter.Age, B=-250)
        2=?
        3=?
        Index=$Particles.PartitionParticles.Partition
        Num=::NewEnumerator0
       )
       From=?
       To=?
       Stabilization Mode:ENiagara_OrientVectorCount=::NewEnumerator0
       Side Direction:Vector3f=?
       Up Direction:Vector3f=?
       Angle Type=::NewEnumerator0
       Normalize:NiagaraBool=true
      )
     )
     Lerp Factor:NiagaraFloat=$User.ParaGlideLeanAlpha
     Enable Spherical Linear Interpolation:NiagaraBool=false
    )
    Particles.CheeseOrientation:Quat4f=&LerpQuaternion(
     Quaternion A:Quat4f=&MultiplyQuaternion(
      Quaternion A=&MakeQuaternion(
       Quaternion From=0
       Axis=<0 1 0>
       Angle=0
       From=?
       To=?
       Stabilization Mode:ENiagara_OrientVectorCount=::NewEnumerator0
       Side Direction:Vector3f=?
       Up Direction:Vector3f=?
       Angle Type=::NewEnumerator0
       Normalize:NiagaraBool=true
      )
      Quaternion B=&MakeQuaternion(
       Quaternion From=0
       Axis=<0 0 1>
       Angle=&MergeHelper_FloatSwitch(
        0=&Multiply_Float(A=$Emitter.Age, B=-30)
        1=&Multiply_Float(A=$Emitter.Age, B=30)
        2=?
        3=?
        Index=$Particles.PartitionParticles.Partition
        Num=::NewEnumerator0
       )
       From=?
       To=?
       Stabilization Mode:ENiagara_OrientVectorCount=::NewEnumerator0
       Side Direction:Vector3f=?
       Up Direction:Vector3f=?
       Angle Type=::NewEnumerator0
       Normalize:NiagaraBool=true
      )
     )
     Quaternion B:Quat4f=&MultiplyQuaternion(
      Quaternion A=&MakeQuaternion(
       Quaternion From=0
       Axis=<0 1 0>
       Angle=&MergeHelper_FloatSwitch(0=-6.5, 1=6.5, 2=?, 3=?, Index=$Particles.PartitionParticles.Partition, Num=::NewEnumerator0)
       From=?
       To=?
       Stabilization Mode:ENiagara_OrientVectorCount=::NewEnumerator0
       Side Direction:Vector3f=?
       Up Direction:Vector3f=?
       Angle Type=::NewEnumerator0
       Normalize:NiagaraBool=true
      )
      Quaternion B=&MakeQuaternion(
       Quaternion From=0
       Axis=<0 0 1>
       Angle=&MergeHelper_FloatSwitch(
        0=&Multiply_Float(A=$Emitter.Age, B=-80)
        1=&Multiply_Float(A=$Emitter.Age, B=80)
        2=?
        3=?
        Index=$Particles.PartitionParticles.Partition
        Num=::NewEnumerator0
       )
       From=?
       To=?
       Stabilization Mode:ENiagara_OrientVectorCount=::NewEnumerator0
       Side Direction:Vector3f=?
       Up Direction:Vector3f=?
       Angle Type=::NewEnumerator0
       Normalize:NiagaraBool=true
      )
     )
     Lerp Factor:NiagaraFloat=$User.ParaGlideLeanAlpha
     Enable Spherical Linear Interpolation:NiagaraBool=false
    )
    Particles.LettuceOrientation:Quat4f=&LerpQuaternion(
     Quaternion A:Quat4f=&MultiplyQuaternion(
      Quaternion A=&MakeQuaternion(
       Quaternion From=0
       Axis=<0 1 0>
       Angle=0
       From=?
       To=?
       Stabilization Mode:ENiagara_OrientVectorCount=::NewEnumerator0
       Side Direction:Vector3f=?
       Up Direction:Vector3f=?
       Angle Type=::NewEnumerator0
       Normalize:NiagaraBool=true
      )
      Quaternion B=&MakeQuaternion(
       Quaternion From=0
       Axis=<0 0 1>
       Angle=&MergeHelper_FloatSwitch(
        0=&Multiply_Float(A=$Emitter.Age, B=40)
        1=&Multiply_Float(A=$Emitter.Age, B=-40)
        2=?
        3=?
        Index=$Particles.PartitionParticles.Partition
        Num=::NewEnumerator0
       )
       From=?
       To=?
       Stabilization Mode:ENiagara_OrientVectorCount=::NewEnumerator0
       Side Direction:Vector3f=?
       Up Direction:Vector3f=?
       Angle Type=::NewEnumerator0
       Normalize:NiagaraBool=true
      )
     )
     Quaternion B:Quat4f=&MultiplyQuaternion(
      Quaternion A=&MakeQuaternion(
       Quaternion From=0
       Axis=<0 1 0>
       Angle=&MergeHelper_FloatSwitch(0=9, 1=-9, 2=?, 3=?, Index=$Particles.PartitionParticles.Partition, Num=::NewEnumerator0)
       From=?
       To=?
       Stabilization Mode:ENiagara_OrientVectorCount=::NewEnumerator0
       Side Direction:Vector3f=?
       Up Direction:Vector3f=?
       Angle Type=::NewEnumerator0
       Normalize:NiagaraBool=true
      )
      Quaternion B=&MakeQuaternion(
       Quaternion From=0
       Axis=<0 0 1>
       Angle=&MergeHelper_FloatSwitch(
        0=&Multiply_Float(A=$Emitter.Age, B=185)
        1=&Multiply_Float(A=$Emitter.Age, B=-185)
        2=?
        3=?
        Index=$Particles.PartitionParticles.Partition
        Num=::NewEnumerator0
       )
       From=?
       To=?
       Stabilization Mode:ENiagara_OrientVectorCount=::NewEnumerator0
       Side Direction:Vector3f=?
       Up Direction:Vector3f=?
       Angle Type=::NewEnumerator0
       Normalize:NiagaraBool=true
      )
     )
     Lerp Factor:NiagaraFloat=$User.ParaGlideLeanAlpha
     Enable Spherical Linear Interpolation:NiagaraBool=false
    )
    Particles.PattyOrientation:Quat4f=&LerpQuaternion(
     Quaternion A:Quat4f=&MultiplyQuaternion(
      Quaternion A=&MakeQuaternion(
       Quaternion From=0
       Axis=<0 1 0>
       Angle=0
       From=?
       To=?
       Stabilization Mode:ENiagara_OrientVectorCount=::NewEnumerator0
       Side Direction:Vector3f=?
       Up Direction:Vector3f=?
       Angle Type=::NewEnumerator0
       Normalize:NiagaraBool=true
      )
      Quaternion B=&MakeQuaternion(
       Quaternion From=0
       Axis=<0 0 1>
       Angle=&MergeHelper_FloatSwitch(
        0=&Multiply_Float(A=$Emitter.Age, B=15)
        1=&Multiply_Float(A=$Emitter.Age, B=-15)
        2=?
        3=?
        Index=$Particles.PartitionParticles.Partition
        Num=::NewEnumerator0
       )
       From=?
       To=?
       Stabilization Mode:ENiagara_OrientVectorCount=::NewEnumerator0
       Side Direction:Vector3f=?
       Up Direction:Vector3f=?
       Angle Type=::NewEnumerator0
       Normalize:NiagaraBool=true
      )
     )
     Quaternion B:Quat4f=&MultiplyQuaternion(
      Quaternion A=&MakeQuaternion(
       Quaternion From=0
       Axis=<0 1 0>
       Angle=&MergeHelper_FloatSwitch(0=12, 1=-12, 2=?, 3=?, Index=$Particles.PartitionParticles.Partition, Num=::NewEnumerator0)
       From=?
       To=?
       Stabilization Mode:ENiagara_OrientVectorCount=::NewEnumerator0
       Side Direction:Vector3f=?
       Up Direction:Vector3f=?
       Angle Type=::NewEnumerator0
       Normalize:NiagaraBool=true
      )
      Quaternion B=&MakeQuaternion(
       Quaternion From=0
       Axis=<0 0 1>
       Angle=&MergeHelper_FloatSwitch(
        0=&Multiply_Float(A=$Emitter.Age, B=175.2272)
        1=&Multiply_Float(A=$Emitter.Age, B=-175.2272)
        2=?
        3=?
        Index=$Particles.PartitionParticles.Partition
        Num=::NewEnumerator0
       )
       From=?
       To=?
       Stabilization Mode:ENiagara_OrientVectorCount=::NewEnumerator0
       Side Direction:Vector3f=?
       Up Direction:Vector3f=?
       Angle Type=::NewEnumerator0
       Normalize:NiagaraBool=true
      )
     )
     Lerp Factor:NiagaraFloat=$User.ParaGlideLeanAlpha
     Enable Spherical Linear Interpolation:NiagaraBool=false
    )
    Particles.PickleOrientation:Quat4f=&LerpQuaternion(
     Quaternion A:Quat4f=&MultiplyQuaternion(
      Quaternion A=&MakeQuaternion(
       Quaternion From=0
       Axis=<0 1 0>
       Angle=0
       From=?
       To=?
       Stabilization Mode:ENiagara_OrientVectorCount=::NewEnumerator0
       Side Direction:Vector3f=?
       Up Direction:Vector3f=?
       Angle Type=::NewEnumerator0
       Normalize:NiagaraBool=true
      )
      Quaternion B=&MultiplyQuaternion(
       Quaternion A=&MakeQuaternion(
        Quaternion From=0
        Axis=<0 1 0>
        Angle=&MergeHelper_FloatSwitch(0=12, 1=-12, 2=?, 3=?, Index=$Particles.PartitionParticles.Partition, Num=::NewEnumerator0)
        From=?
        To=?
        Stabilization Mode:ENiagara_OrientVectorCount=::NewEnumerator0
        Side Direction:Vector3f=?
        Up Direction:Vector3f=?
        Angle Type=::NewEnumerator0
        Normalize:NiagaraBool=true
       )
       Quaternion B=&MakeQuaternion(
        Quaternion From=0
        Axis=<0 0 1>
        Angle=&MergeHelper_FloatSwitch(
         0=&Multiply_Float(A=$Emitter.Age, B=-250)
         1=&Multiply_Float(A=$Emitter.Age, B=250)
         2=?
         3=?
         Index=$Particles.PartitionParticles.Partition
         Num=::NewEnumerator0
        )
        From=?
        To=?
        Stabilization Mode:ENiagara_OrientVectorCount=::NewEnumerator0
        Side Direction:Vector3f=?
        Up Direction:Vector3f=?
        Angle Type=::NewEnumerator0
        Normalize:NiagaraBool=true
       )
      )
     )
     Quaternion B:Quat4f=&MultiplyQuaternion(
      Quaternion A=&MakeQuaternion(
       Quaternion From=0
       Axis=<0 1 0>
       Angle=&MergeHelper_FloatSwitch(0=20, 1=-20, 2=?, 3=?, Index=$Particles.PartitionParticles.Partition, Num=::NewEnumerator0)
       From=?
       To=?
       Stabilization Mode:ENiagara_OrientVectorCount=::NewEnumerator0
       Side Direction:Vector3f=?
       Up Direction:Vector3f=?
       Angle Type=::NewEnumerator0
       Normalize:NiagaraBool=true
      )
      Quaternion B=&MakeQuaternion(
       Quaternion From=0
       Axis=<0 0 1>
       Angle=&MergeHelper_FloatSwitch(
        0=&Multiply_Float(A=$Emitter.Age, B=-250)
        1=&Multiply_Float(A=$Emitter.Age, B=250)
        2=?
        3=?
        Index=$Particles.PartitionParticles.Partition
        Num=::NewEnumerator0
       )
       From=?
       To=?
       Stabilization Mode:ENiagara_OrientVectorCount=::NewEnumerator0
       Side Direction:Vector3f=?
       Up Direction:Vector3f=?
       Angle Type=::NewEnumerator0
       Normalize:NiagaraBool=true
      )
     )
     Lerp Factor:NiagaraFloat=$User.ParaGlideLeanAlpha
     Enable Spherical Linear Interpolation:NiagaraBool=false
    )
    Particles.TomatoOrientation:Quat4f=&LerpQuaternion(
     Quaternion A:Quat4f=&MultiplyQuaternion(
      Quaternion A=&MakeQuaternion(
       Quaternion From=0
       Axis=<0 1 0>
       Angle=0
       From=?
       To=?
       Stabilization Mode:ENiagara_OrientVectorCount=::NewEnumerator0
       Side Direction:Vector3f=?
       Up Direction:Vector3f=?
       Angle Type=::NewEnumerator0
       Normalize:NiagaraBool=true
      )
      Quaternion B=&MakeQuaternion(
       Quaternion From=0
       Axis=<0 0 1>
       Angle=&MergeHelper_FloatSwitch(
        0=&Multiply_Float(A=$Emitter.Age, B=-40)
        1=&Multiply_Float(A=$Emitter.Age, B=40)
        2=?
        3=?
        Index=$Particles.PartitionParticles.Partition
        Num=::NewEnumerator0
       )
       From=?
       To=?
       Stabilization Mode:ENiagara_OrientVectorCount=::NewEnumerator0
       Side Direction:Vector3f=?
       Up Direction:Vector3f=?
       Angle Type=::NewEnumerator0
       Normalize:NiagaraBool=true
      )
     )
     Quaternion B:Quat4f=&MultiplyQuaternion(
      Quaternion A=&MakeQuaternion(
       Quaternion From=0
       Axis=<0 1 0>
       Angle=&MergeHelper_FloatSwitch(0=-6.5, 1=6.5, 2=?, 3=?, Index=$Particles.PartitionParticles.Partition, Num=::NewEnumerator0)
       From=?
       To=?
       Stabilization Mode:ENiagara_OrientVectorCount=::NewEnumerator0
       Side Direction:Vector3f=?
       Up Direction:Vector3f=?
       Angle Type=::NewEnumerator0
       Normalize:NiagaraBool=true
      )
      Quaternion B=&MakeQuaternion(
       Quaternion From=0
       Axis=<0 0 1>
       Angle=&MergeHelper_FloatSwitch(
        0=&Multiply_Float(A=$Emitter.Age, B=-250)
        1=&Multiply_Float(A=$Emitter.Age, B=250)
        2=?
        3=?
        Index=$Particles.PartitionParticles.Partition
        Num=::NewEnumerator0
       )
       From=?
       To=?
       Stabilization Mode:ENiagara_OrientVectorCount=::NewEnumerator0
       Side Direction:Vector3f=?
       Up Direction:Vector3f=?
       Angle Type=::NewEnumerator0
       Normalize:NiagaraBool=true
      )
     )
     Lerp Factor:NiagaraFloat=$User.ParaGlideLeanAlpha
     Enable Spherical Linear Interpolation:NiagaraBool=false
    )
   ) {isSetParameters=true}
   &UpdateMeshOrientation(
    Facing Direction=$System.FN_Contrail_Setup.Direction Vector
    Facing Interpolation=false
    Orientation Method=::NewEnumerator1
    Override Quat=false
    Stabilization Mode:ENiagara_OrientVectorCount=::NewEnumerator2
    TransformFacing=false
    TransformSide=false
   )
   &UpdateMeshOrientation(
    Rotation Rate=0.125
    Rotation Vector=&MakeVector(X=0, Y=0, Z=&RandomRangeFloat(Minimum=-1, Maximum=1))
   ) {module=UpdateMeshOrientation001}
   &AlignSpriteToMeshOrientation(Mesh Orientation Relative Sprite Alignment Vector=<-1 0 0>, SetAlignment=true, SetFacing=false) {enabled=false}
   &ScaleColor(
    Scale RGB=&MakeVectorFromLinearColorRGB(Color=&FNTimeOfDayColor(Night Color=<rgba 0.6302 0.7887 1 1>, Morning Color=<rgba 1 1 1 1>, Day Color=<rgba 1 1 1 1>, Evening Color=<rgba 1 1 1 1>))
    ScaleA=true
    ScaleRGB=true
   )
   /SetVariables(
    Particles.Scale=&VectorFromFloat(Value=&FloatFromCurve(
      FloatCurve=DataInterface() {propertyValues {Curve {
         defaultValue=1
         keys [{arriveTangent=3.0116, interpMode=RCIM_Cubic, time=0, value=0}, {interpMode=RCIM_Cubic, leaveTangent=0.4934, time=0.75, value=1}]
        }}}
      CurveIndex=$Emitter.Age
     ))
    Particles.Tomato Color:LinearColor=&MakeLinearColorFromVectorAndFloat(
     "Vector (RGB)"=&MakeVectorFromLinearColorRGB(Color=&FNTimeOfDayColor(Night Color=<rgba 0.1834 0.2687 0.8385 1>, Morning Color=<rgba 1 1 1 1>, Day Color=<rgba 1 1 1 1>, Evening Color=<rgba 1 1 1 1>))
     "Float (Alpha)"=1
    )
   ) {isSetParameters=true}
  ]
  rendererClasses [&NiagaraMeshRendererProperties, &NiagaraSpriteRendererProperties]
  renderers [
   {rendererClass=&NiagaraMeshRendererProperties, rendererIndex=0}
   {rendererClass=&NiagaraSpriteRendererProperties, rendererIndex=1}
   {rendererClass=&NiagaraMeshRendererProperties, rendererIndex=2}
   {rendererClass=&NiagaraMeshRendererProperties, rendererIndex=3}
   {rendererClass=&NiagaraMeshRendererProperties, rendererIndex=4}
   {rendererClass=&NiagaraMeshRendererProperties, rendererIndex=5}
   {rendererClass=&NiagaraMeshRendererProperties, rendererIndex=6}
   {rendererClass=&NiagaraMeshRendererProperties, rendererIndex=7}
   {rendererClass=&NiagaraMeshRendererProperties, rendererIndex=8}
  ]
 }
 MustardMesh {
  emitterUpdateScript [
   &EmitterState(Life Cycle Mode=::NewEnumerator1, Loop Behavior=::NewEnumerator1, Loop Duration Mode=::NewEnumerator1, UseLoopDelay=false)
   &SpawnRate(SpawnRate=&Lerp_Float(A=12, B=16, Alpha=$User.ParaGlideLeanAlpha), Use Spawn Probability=false)
   /SetVariables(Emitter.UVTiling:Vector4f=&MakeVector4(
     X=&Lerp_Float(
      A=&RandomRangeFloat(Minimum=1, Maximum=0.3)
      B=&RandomRangeFloat(Minimum=2, Maximum=1)
      Alpha=$User.ParaGlideLeanAlpha
     )
     Y=&RandomRangeFloat(Minimum=0.7, Maximum=1)
     Z=1
     W=1
    )) {isSetParameters=true, module=SetVariables_#2}
  ]
  particleSpawnScript [
   &PartitionParticles()
   &InitializeParticle(
    Color=&FNTimeOfDayColor(Night Color=<rgba 2 0.8978 0 1>, Morning Color=<rgba 1 0.4489 0 1>, Day Color=<rgba 1 0.8348 0 1>, Evening Color=<rgba 1 0.4489 0 1>)
    Lifetime=&Lerp_Float(A=1.2, B=1, Alpha=$User.ParaGlideLeanAlpha)
    Mesh Scale Mode=::NewEnumerator3
    UsePositionOffset=false
    Write Material Random=true
   )
   &InitialMeshOrientation(Mesh Orientation Mode=::NewEnumerator1, Use Rotation Vector=true) {enabled=false}
   &InitialMeshOrientation(
    Mesh Orientation Mode=::NewEnumerator1
    Rotation=&MaskVector(Mask Source=1, Mask Parameter=$Particles.PartitionParticles.Partition, Mask=0, Masked Vector=<0 0 0.304>, Passthrough Vector=<0 0 -0.28>)
    Use Rotation Vector=true
   ) {module=InitialMeshOrientation001}
   &InitialMeshOrientation(
    Mesh Orientation Mode=::NewEnumerator1
    Rotation=&MaskVector(Mask Source=1, Mask Parameter=$Particles.PartitionParticles.Partition, Mask=0, Masked Vector=<0 0.01 0>, Passthrough Vector=<0 -0.01 0>)
    Use Rotation Vector=true
   ) {module=InitialMeshOrientation002}
  ]
  particleUpdateScript [
   &ParticleState()
   &SkeletalMeshLocation(
    Bone / Socket Index=&SelectIntFromArray(Direct Array Index=$Particles.PartitionParticles.Partition, Int Selection Array=DataInterface() {propertyValues {IntData [0,1]}})
    MaskSpawnGroup=false
    Mesh Sampling Type=::NewEnumerator5
    Orientation Sampling=::NewEnumerator1
    Sampled Position Offset=&AddVectorToPosition(
     Position=&Lerp_Vector(
      A=&MergeHelper_VectorSwitch(0=<0 0 27.5>, 1=<0 0 27.5>, 2=?, 3=?, Index=$Particles.PartitionParticles.Partition, Num=::NewEnumerator0)
      B=&MergeHelper_VectorSwitch(0=<8 0.001 88.5>, 1=<-8 0 88.5>, 2=?, 3=?, Index=$Particles.PartitionParticles.Partition, Num=::NewEnumerator0)
      Alpha=$User.ParaGlideLeanAlpha
     )
     Vector=<0 0 0>
    )
    Sampled Position Offset Coordinate Space:ENiagaraSkelMeshTransforms=::NewEnumerator1
    Skeletal Mesh=$System.Skeletal Mesh
    Socket Sampling Mode=::NewEnumerator5
   )
   &ScaleMeshSize(Scale Factor=&Lerp_Vector(
     A=&MergeHelper_VectorSwitch(0=<1.4 1.4 0.125>, 1=<1.4 1.4 0.15>, 2=?, 3=?, Index=$Particles.PartitionParticles.Partition, Num=::NewEnumerator0)
     B=&MergeHelper_VectorSwitch(0=<1.4 1.4 0.3>, 1=<1.4 1.4 0.35>, 2=?, 3=?, Index=$Particles.PartitionParticles.Partition, Num=::NewEnumerator0)
     Alpha=$User.ParaGlideLeanAlpha
    )) {module=ScaleMeshSize002}
   &ScaleMeshSize(Scale Factor=&VectorFromCurve(
     VectorCurve=DataInterface() {propertyValues {
       XCurve {defaultValue=0.2, keys [{arriveTangent=3.0023, time=0, value=0.6}, <at 1 1>]}
       YCurve {keys [{arriveTangent=0.803, time=0, value=0.6}, {arriveTangent=0.465, leaveTangent=0.465, time=1, value=0.8}]}
       ZCurve {
        defaultValue=0.2
        keys [
         {arriveTangent=3.0023, interpMode=RCIM_Cubic, leaveTangent=3.0023, tangentMode=RCTM_User, time=0, value=0.3}
         {arriveTangentWeight=0.1667, interpMode=RCIM_Cubic, tangentMode=RCTM_User, time=0.5, value=1}
         {arriveTangent=-5.505, arriveTangentWeight=0.1667, interpMode=RCIM_Cubic, leaveTangent=-5.505, tangentMode=RCTM_User, time=1, value=0}
        ]
       }
      }}
     Scale Curve=&VectorFromFloat(Value=1)
    )) {enabled=false, module=ScaleMeshSize001}
   &ScaleMeshSize(Scale Factor=&Multiply_Vector(
     A=&VectorFromCurve(VectorCurve=DataInterface() {propertyValues {
        XCurve {defaultValue=0.2, keys [{arriveTangent=3.0023, time=0, value=0.8}, <at 1 1>]}
        YCurve {keys [{arriveTangent=0.803, time=0, value=0.8}, {arriveTangent=0.465, leaveTangent=0.465, time=1, value=1}]}
        ZCurve {
         defaultValue=0.2
         keys [
          {arriveTangent=3.0023, interpMode=RCIM_Cubic, leaveTangent=3.0023, tangentMode=RCTM_User, time=0, value=0.3}
          {arriveTangentWeight=0.1667, interpMode=RCIM_Cubic, tangentMode=RCTM_User, time=0.5, value=1}
          {arriveTangent=-5.505, arriveTangentWeight=0.1667, interpMode=RCIM_Cubic, leaveTangent=-5.505, tangentMode=RCTM_User, time=1, value=0}
         ]
        }
       }})
     B=&VectorFromFloat(Value=&FloatFromCurve(
       FloatCurve=DataInterface() {propertyValues {Curve {
          defaultValue=1
          keys [{arriveTangent=3.0116, interpMode=RCIM_Cubic, time=0, value=0}, {interpMode=RCIM_Cubic, leaveTangent=0.4934, time=0.75, value=1}]
         }}}
       CurveIndex=$Emitter.Age
      ))
    ))
   &ScaleColor(
    Scale Alpha=&FloatFromCurve(
     FloatCurve=DataInterface() {propertyValues {Curve {keys [
         <at 0 1 RCIM_Cubic>
         {arriveTangent=-0.8638, interpMode=RCIM_Cubic, leaveTangent=-0.8638, tangentMode=RCTM_User, time=0.7618, value=0.3384}
         {arriveTangent=-0.4827, interpMode=RCIM_Cubic, leaveTangent=-0.4827, tangentMode=RCTM_User, time=1, value=0.1}
        ]}}}
     Scale Curve=&Lerp_Float(
      A=0
      B=1
      Alpha=&RemapRange(Input Value To Remap=$User.ParaGlideLeanAlpha, Input Min=0.16, Input Max=1, Output Min=0, Output Max=1)
     )
    )
    Scale RGB=&MakeVectorFromLinearColorRGB(Color=&FNTimeOfDayColor(Night Color=<rgba 0.4271 0.5366 1 1>, Morning Color=<rgba 1 1 1 1>, Day Color=<rgba 1 1 1 1>, Evening Color=<rgba 1 1 1 1>))
    ScaleA=true
    ScaleRGB=true
   )
   &DynamicMaterialParameters(
    Index 0 Param 2=&Lerp_Float(A=0, B=5, Alpha=$User.ParaGlideLeanAlpha)
    Index 0 Param 3=&RandomRangeFloat(Minimum=0, Maximum=1)
    Module.Index0Float0Write=true
    Module.Index0Float1Write=true
    Module.Index0Float2Write=true
    Module.Index0Float3Write=true
   )
  ]
  rendererClasses [&NiagaraMeshRendererProperties]
  renderers [{rendererClass=&NiagaraMeshRendererProperties, rendererIndex=0}]
 }
}
summary {
 systemName=NS_Contrail_BoarGore
 emitters [
  {
   emitterName=Mesh_SkeletalMesh
   rendererClasses [&NiagaraMeshRendererProperties, &NiagaraSpriteRendererProperties]
  }
  {
   emitterName=KetchupMustardBits
   rendererClasses [&NiagaraMeshRendererProperties, &NiagaraSpriteRendererProperties]
  }
  {emitterName=FoodBits, rendererClasses [&NiagaraMeshRendererProperties]}
  {emitterName=GreaseMesh, rendererClasses [&NiagaraMeshRendererProperties]}
  {emitterName=KetchupMesh, rendererClasses [&NiagaraMeshRendererProperties]}
  {emitterName=MustardMesh, rendererClasses [&NiagaraMeshRendererProperties]}
  {bEnabled=false, emitterName=FN_Debug, rendererClasses [&NiagaraSpriteRendererProperties]}
 ]
 userVariables [
  User.bIsFrontEnd {type=NiagaraBool}
  User.bIsFrontEndPreview {type=NiagaraBool}
  User.ParaGlideLeanAlpha {type=NiagaraFloat, defaultValue=1, description=""}
  User.RibbonIDOffset {type=NiagaraInt32}
 ]
}
